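# Unlit additive material for Niagara sprites that draws glowing rays for the stars aura cosmetic.
# It masks the rays texture with an exponential radial gradient, then adds a smaller second gradient as a central glow.
# Particle color tints the result, which passes through a shared emissive adjustment function.
# Texture alpha is multiplied by particle alpha.
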Material /SparksCosmetics/Items/Auras/Assets/NS_Aura_Stars/M_Aura_Rays
graph {
 properties {
  bCastRayTracedShadows=false
  bUsedWithNiagaraSprites=true
  blendMode=BLEND_Additive
  pixelDepthOffsetMode=PDOM_Legacy
  shadingModel=MSM_Unlit
  usedShadingModels=Unlit
 }
 nodes [
  n1 = TextureSample() {texture=&/T_Aura_Rays}
  n2 = Multiply(
   A=MaterialFunctionCall(Radius=Constant() {r=0.6}) {materialFunction=&RadialGradientExponential}.RadialGradientExponential
   B=@n1.RGB
  )
  n3 = MaterialFunctionCall(Radius=Constant() {r=0.26}, Density=Constant() {r=0.3}) {materialFunction=&RadialGradientExponential}
  n4 = ParticleColor()
  MaterialFunctionCall(Emissive=Multiply(A=Add(A=@n2, B=@n3.RadialGradientExponential), B=@n4.RGB)) {materialFunction=&MF_FX_EmissiveAdjustments}
  Multiply(A=@n1.A, B=@n4.A)
 ]
}
