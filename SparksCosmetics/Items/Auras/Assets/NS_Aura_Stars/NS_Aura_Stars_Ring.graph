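# Additive unlit Niagara sprite material that draws a procedural ring for the Stars aura cosmetic effect.
# It subtracts a smaller exponential radial gradient from a slightly larger one to carve the ring without textures.
# Particle color RGB tints the ring, which then feeds the emissive input of the MF_FX_EmissiveAdjustments function.
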Material /SparksCosmetics/Items/Auras/Assets/NS_Aura_Stars/NS_Aura_Stars_Ring
graph {
 properties {
  bCastRayTracedShadows=false
  bUsedWithNiagaraSprites=true
  blendMode=BLEND_Additive
  pixelDepthOffsetMode=PDOM_Legacy
  shadingModel=MSM_Unlit
  usedShadingModels=Unlit
 }
 nodes [
  Constant() {r=2.2}
  n1 = MaterialFunctionCall(Radius=Constant() {r=0.45}) {materialFunction=&RadialGradientExponential}
  n2 = MaterialFunctionCall(Radius=Constant() {r=0.4}, Density=Constant() {r=4}) {materialFunction=&RadialGradientExponential}
  MaterialFunctionCall(Emissive=Multiply(A=Subtract(A=@n1.RadialGradientExponential, B=@n2.RadialGradientExponential), B=ParticleColor().RGB)) {materialFunction=&MF_FX_EmissiveAdjustments}
 ]
}
